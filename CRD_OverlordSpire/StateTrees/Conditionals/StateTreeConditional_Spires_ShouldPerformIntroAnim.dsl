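; StateTree conditional for the Overlord Spire that decides whether the spire should perform its intro animation.
; Its test function gets the spire device from the context and returns the result of ShouldPerformIntroSequence on it.
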(fn ReceiveTestCondition
  (return (StateTreeConditions|ShouldPerformIntroSequence (Variables|Context|GetSpireDevice))))
